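# Sound cue for the end-of-dial sound of the rotary phone quest in Figment Season 8 Athena content.
# A Random node picks between two dial-end wave variants with equal weight and feeds a Modulator at the root, which is left at default settings.
# It borrows the gadget-activate attenuation asset for spatial falloff and routes through the mission in-game sound class.
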SoundCue /Figment_S08/Athena/Sounds/Quests/RotaryPhone/Athena_Quest_Rotary_DialEnd_Cue
root=Modulator_0
nodes [
 {id=Modulator_0, children [Random_0]}
 {id=Random_0, children [WavePlayer_1,WavePlayer_2], properties {weights [1,1]}}
 {id=WavePlayer_1, properties {soundWaveAssetPtr=&/Athena_Quest_Rotary_DialEnd_01}}
 {id=WavePlayer_2, properties {soundWaveAssetPtr=&/Athena_Quest_Rotary_DialEnd_02}}
]
properties {
 attenuationSettings=&Fort_GadgetActivate_Att
 maxDistance=4000
 soundClassObject=/Game/Sounds/FortSoundClassesAndModes/Fort_Mission_IngameSounds
 volumeMultiplier=1
}
